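# Niagara aura cosmetic "Center Stage" that surrounds a character with confetti, flares and glows.
# Four sprite emitters sample the System skeletal mesh: velocity-inheriting confetti, drifting cross and large flares, and looping glow decals.
# Partition indices select per-particle bones and offsets; flares smooth their follow with InterpolateOverTime, and a center glow position is lerped between two bones.
# Front-end, large-body and time-of-day inputs adjust brightness, size or camera offset per emitter.
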NiagaraSystem /SparksCosmetics/Items/Auras/Assets/NS_Aura_CenterStage/NS_Aura_CenterStage
dependencies {
 usedDataInterfaces [
  &NiagaraDataInterfaceArrayColor
  &NiagaraDataInterfaceCurve
  &NiagaraDataInterfaceArrayFloat
  &NiagaraDataInterfaceArrayInt32
  &NiagaraDataInterfaceArrayFloat3
  &NiagaraDataInterfaceVectorCurve
  &NiagaraDataInterfaceColorCurve
  &NiagaraDataInterfaceArrayFloat2
  &NiagaraDataInterfaceSkeletalMesh
 ]
 usedDynamicInputs []
 usedModules [
  &self:SystemSpawnScript.SystemScriptSource.SystemScriptGraph.NiagaraNodeAssignment_0.SetVariables_#1
  &SystemState
 ]
}
emitters {
 Confetti {
  emitterUpdateScript [
   &EmitterState()
   &SpawnRate(SpawnRate=&RandomRangeFloat(Minimum=25, Maximum=30), Use Spawn Probability=false)
  ]
  particleSpawnScript [
   &InitializeParticle(
    Color=&SelectLinearColorFromArray(
     Array Sampling Mode=::NewEnumerator0
     Direct Array Index=?
     Color Selection Array=DataInterface() {propertyValues {ColorData [<rgba 1 0.2436 0.1223 1>, <rgba 1 0.6709 0.1198 1>, <rgba 1 0.0885 0.0729 1>, <rgba 1 0.375 0.5707 1>, <rgba 0 1 0.2463 1>]}}
     Random Seed=0
    )
    Lifetime=&RandomRangeFloat(Minimum=0.9, Maximum=1.5)
    Sprite Rotation Mode=::NewEnumerator2
    Sprite Size Mode=::NewEnumerator1
    Uniform Sprite Size Max=12
    UsePositionOffset=false
    Write Material Random=true
   )
   &ShapeLocation(
    Apply To Particle Position=::NewEnumerator1
    MaskSpawnGroup=false
    Override Seed=false
    Shape Origin=&AddVectorToPosition(Position=<0 0 0>, Vector=<10 0 0>)
    Sphere Radius=65
    UseOwnerScale=true
   )
   &SkeletalMeshLocation(
    Bone Sampling Mode=::NewEnumerator4
    MaskSpawnGroup=false
    Sampled Position Offset:NiagaraPosition=$StackContext.ShapeLocation.ShapePosition
    Skeletal Mesh=$System.Skeletal Mesh
   ) {module=SkeletalMeshLocation001}
   &CameraOffset(Camera Offset Amount=&MakeCustomFloatFromBool(
     Bool=$User.IsFrontEnd
     True Float=&MakeCustomFloatFromBool(
      Bool=&SetBoolByFloatComparison(A=&MakeFloatFromVector2d(Vector 2D=$Particles.SpriteSize), B=7)
      True Float=-5
      False Float=20
     )
     False Float=&MakeCustomFloatFromBool(
      Bool=&SetBoolByFloatComparison(A=&MakeFloatFromVector2d(Vector 2D=$Particles.SpriteSize), B=7)
      True Float=-25
      False Float=25
     )
    ))
   &AddVelocity(Velocity=<0 0 -1>, Velocity Speed Scale=&RandomRangeFloat(Minimum=0, Maximum=5)) {module=AddVelocity001}
   &InheritVelocity(
    Inherited Velocity Amount Scale=&Multiply_VectorByFloat(Vector=<1 1 1>, Float=300)
    Inherited Velocity Speed Limit=150
    Speed Limit=true
    UseSpeedThreshold=false
   )
   &ApplyInitialForces()
  ]
  particleUpdateScript [
   &ParticleState()
   &ScaleColor(
    Scale RGB=&MakeCustomVectorFromBool(
     Bool=$User.IsFrontEnd
     True Vector=&Multiply_VectorByFloat(Vector=<1 0.95 0.85>, Float=1.5)
     False Vector=&VectorFromFloat(Value=1)
    )
    ScaleA=true
    ScaleRGB=true
   )
   &Drag(Drag=&RandomRangeFloat(Minimum=1.5, Maximum=5), Use Linear Drag=true, Use Rotational Drag=true)
   &GravityForce(Gravity=&RandomRangeVector(Minimum=<0 0 -50>, Maximum=<0 0 -250>))
   &CurlNoiseForce(
    Bias Noise=false
    Noise Frequency=25
    Noise Strength=&RandomRangeFloat(Minimum=50, Maximum=200)
    Pan Noise=false
   )
   &ScaleSpriteSize(Scale Factor=&Vector2DFromFloat(Value=&FloatFromCurve(FloatCurve=DataInterface() {propertyValues {Curve {
         defaultValue=0
         keys [
          {interpMode=RCIM_Cubic, tangentMode=RCTM_User, time=0, value=0}
          {interpMode=RCIM_Cubic, tangentMode=RCTM_User, time=0.3, value=0.8}
          {tangentMode=RCTM_User, time=1, value=0}
         ]
        }}})))
   /Niagara/Modules/Update/SubUV/SubUVAnimation(Number Of Frames=36)
   &SpriteRotationRate(Rotation Rate=&RandomRangeFloat(Minimum=400, Maximum=100)) {module=SpriteRotationRate001}
   &SolveForcesAndVelocity(Clamp Velocity=false, Limit Acceleration=false)
  ]
  rendererClasses [&NiagaraSpriteRendererProperties]
  renderers [{rendererClass=&NiagaraSpriteRendererProperties}]
 }
 Cross_Flares {
  emitterUpdateScript [
   &EmitterState()
   &SpawnRate(Spawn Probability=0.5, SpawnRate=0.25, Use Spawn Probability=true) {module=SpawnRate001}
   &SpawnRate(Spawn Probability=0.8, SpawnRate=3, Use Spawn Probability=false)
   &SpawnBurst_Instantaneous(Spawn Count=2, Use Spawn Probability=false, UseLoopCountLimit=true)
  ]
  particleSpawnScript [
   &PartitionParticles(Number Of Partitions=8)
   &InitializeParticle(
    Color=&SelectLinearColorFromArray(
     Array Sampling Mode=::NewEnumerator1
     Direct Array Index=0
     Color Selection Array=DataInterface() {propertyValues {ColorData [<rgba 1.5 0.6917 0.4856 1>, <rgba 1.5 1.1073 0.4885 1>, <rgba 1.5 0.6594 0.734 1>]}}
     Random Seed=?
    )
    Lifetime=&SelectFloatFromArray(
     Direct Array Index=$Particles.PartitionParticles.Partition
     Direct Array Mode=::NewEnumerator1
     Float Selection Array=DataInterface() {propertyValues {FloatData [0.7,2.2,1.2,1.95,1.5,0.8]}}
    )
    Sprite Rotation Mode=::NewEnumerator0
    Sprite Size Mode=::NewEnumerator3
    Uniform Sprite Size=&Multiply_Float(
     A=&Multiply_Float(
      A=&SelectFloatFromArray(
       Direct Array Index=$Particles.PartitionParticles.Partition
       Direct Array Mode=::NewEnumerator1
       Float Selection Array=DataInterface() {propertyValues {FloatData [215,275,175,250,225,190]}}
      )
      B=&RandomRangeFloat(Minimum=0.9, Maximum=1)
     )
     B=&MakeCustomFloatFromBool(Bool=$User.IsLargeBody, True Float=1.2, False Float=1)
    )
    UsePositionOffset=false
    Write Material Random=true
   ) {module=InitializeParticle001}
   /SetVariables(Particles.glowScale:Vector2f=&Multiply_Vector2D_ByFloat(Float=0.5, Vector2D=$Particles.SpriteSize)) {isSetParameters=true, module=SetVariables_#2}
   /SetVariables(Particles.glowCameraOffset=-35) {isSetParameters=true, module=SetVariables_#3}
   /SetVariables(Particles.glowcolor:LinearColor=<rgba 0.2423 0.0648 0.0467 1>) {isSetParameters=true, module=SetVariables_#4}
  ]
  particleUpdateScript [
   &ParticleState()
   &ScaleSpriteSize(Uniform Curve Sprite Scale=DataInterface() {propertyValues {Curve {
       defaultValue=0
       keys [
        {arriveTangent=9.0725, interpMode=RCIM_Cubic, leaveTangent=9.0725, tangentMode=RCTM_User, time=0, value=1}
        {arriveTangent=-0.9031, interpMode=RCIM_Cubic, leaveTangent=-0.9031, tangentMode=RCTM_User, time=0.1, value=1.375}
        {arriveTangent=-0.0405, interpMode=RCIM_Cubic, leaveTangent=-0.0405, tangentMode=RCTM_User, time=1, value=1}
       ]
      }}}) {module=ScaleSpriteSize001}
   /SetVariables(Particles.Waveform:NiagaraFloat=&Waveform(
     Global Input Position=&Multiply_Float(A=&ReturnNormalizedExecIndex(), B=10)
     "Waveform [1]"=::NewEnumerator12
     "[1] Phase Scale"=2
     "[1] Frequency"=1
     "[1] Amplitude Min/Max"=&SelectVector2DFromArray(
      Array Sampling Mode=::NewEnumerator0
      Direct Array Index=?
      Vector 2D Selection Array=DataInterface() {propertyValues {FloatData [<0.2 1>, <0.75 1.1>]}}
      Random Seed=0
     )
    )) {isSetParameters=true}
   &DynamicMaterialParameters(
    Index 0 Param 1=&RandomRangeFloat(Minimum=3, Maximum=5)
    Index 0 Param 3=1
    Module.Param0WriteEnabled=true
    Module.Param1WriteEnabled=true
    Module.Param2WriteEnabled=false
    Module.Param3WriteEnabled=false
   )
   &SkeletalMeshLocation(
    Bone / Socket Index=&SelectIntFromArray(Direct Array Index=$Particles.PartitionParticles.Partition, Int Selection Array=DataInterface() {propertyValues {IntData [1,2,3,4,1,2,3,4]}})
    Bone Sampling Mode=::NewEnumerator4
    MaskSpawnGroup=false
    Orientation Sampling=::NewEnumerator1
    Sampled Position Offset=&SelectVectorFromArray(
     Direct Array Index=$Particles.PartitionParticles.Partition
     Vector Selection Array=DataInterface() {propertyValues {FloatData [<9 -10 25>, <-15.9194 -5 20>, <5.0892 2 22>, <-1.8313 -10 13.1757>, <-7.8551 0 27>, <-18 0 9.9419>, <-6.647 -5 -10>, <-4.9501 0 -14.9324>]}}
    )
    Sampled Position Offset Coordinate Space:ENiagaraSkelMeshTransforms=::NewEnumerator0
    Skeletal Mesh=$System.Skeletal Mesh
   ) {module=SkeletalMeshLocation003}
   &InterpolateOverTime(Target Position Value:NiagaraPosition=$Particles.Position, Variable Type=::NewEnumerator6)
   &ScaleColor(
    Scale Alpha=&Multiply_Float(
     A=&FloatFromCurve(FloatCurve=DataInterface() {propertyValues {Curve {
         defaultValue=1
         keys [
          {arriveTangent=4.3601, interpMode=RCIM_Cubic, leaveTangent=4.3601, tangentMode=RCTM_User, time=0, value=0}
          {interpMode=RCIM_Cubic, tangentMode=RCTM_User, time=0.2, value=1}
          {arriveTangentWeight=0.2668, interpMode=RCIM_Cubic, leaveTangentWeight=0.1601, tangentMode=RCTM_User, time=0.4, value=1}
          {tangentMode=RCTM_User, time=1, value=0}
         ]
        }}})
     B=&MakeCustomFloatFromBool(
      Bool=$User.IsFrontEnd
      True Float=1
      False Float=&FNTimeOfDayScalar(Night Scale=0.5, Morning Scale=0.8, Evening Scale=0.6)
     )
    )
    Scale RGB=&Multiply_VectorByFloat(
     Vector=<1 1 1>
     Float=&MakeCustomFloatFromBool(
      Bool=$User.IsFrontEnd
      True Float=1
      False Float=&FNTimeOfDayScalar(Night Scale=0.8, Morning Scale=1, Day Scale=1.2, Evening Scale=0.9)
     )
    )
    ScaleA=true
    ScaleRGB=true
   )
   /SetVariables(Particles.SpriteRotation:NiagaraFloat=&Waveform("Waveform [1]"=::NewEnumerator12, "[1] Frequency"=0.1, "[1] Amplitude Min/Max"=<-1 1>, "[1] Scale"=180)) {isSetParameters=true}
   /SetVariables(Particles.glowcolor:LinearColor=&Multiply_LinearColorByFloat(
     Linear Color=&ColorFromCurve(ColorCurve=DataInterface() {propertyValues {
        AlphaCurve {
         defaultValue=0
         keys [
          {arriveTangent=4.3601, interpMode=RCIM_Cubic, leaveTangent=4.3601, tangentMode=RCTM_User, time=0, value=0}
          {arriveTangentWeight=0.0667, interpMode=RCIM_Cubic, tangentMode=RCTM_User, time=0.2, value=1}
          {arriveTangentWeight=0.0667, interpMode=RCIM_Cubic, leaveTangentWeight=0.2, tangentMode=RCTM_User, time=0.4, value=1}
          <at 1 0>
         ]
        }
        BlueCurve {keys [<at 0.2641 0.0076 RCIM_Cubic>,<at 1 0.091>]}
        GreenCurve {keys [<at 0.2641 0.0297 RCIM_Cubic>,<at 1 0.0281>]}
        RedCurve {keys [<at 0.2641 0.2423 RCIM_Cubic>,<at 1 0.1719>]}
       }})
     Float=&MakeCustomFloatFromBool(
      Bool=$User.IsFrontEnd
      True Float=0.7
      False Float=&FNTimeOfDayScalar(Night Scale=1, Morning Scale=1, Day Scale=0.8, Evening Scale=1)
     )
    )) {isSetParameters=true}
   /SetVariables(Particles.Position=&AddVectorToPosition(
     Position=$StackContext.InterpolateOverTime.InterpolateOverTime_Position.Moving Average
     Vector=&MakeVector(
      X=0
      Y=0
      Z=&FloatFromCurve(
       FloatCurve=DataInterface() {propertyValues {Curve {
          defaultValue=1
          keys [
           {interpMode=RCIM_Cubic, leaveTangentWeight=0.5658, tangentMode=RCTM_User, time=0, value=0}
           {arriveTangentWeight=0.3333, interpMode=RCIM_Cubic, tangentMode=RCTM_User, time=1, value=1}
          ]
         }}}
       Scale Curve=&MakeCustomFloatFromBool(
        Bool=&SetBoolByFloatComparison(A=$Particles.Initial.Lifetime, B=1, Comparison Type=::NewEnumerator2)
        True Float=&SelectFloatFromArray(
         Array Sampling Mode=::NewEnumerator0
         Direct Array Index=?
         Float Selection Array=DataInterface() {propertyValues {FloatData [-5,5,-4]}}
         Random Seed=0
        )
        False Float=0
       )
      )
     )
    )) {isSetParameters=true, module=SetVariables_#5}
   /SetVariables(Particles.MeshOrientation=&MakeQuaternion(
     Quaternion From=0
     Axis=&MakeVector(X=1, Y=0, Z=0)
     Angle=&Waveform("Waveform [1]"=::NewEnumerator12, "[1] Frequency"=0.1, "[1] Amplitude Min/Max"=<-1 1>, "[1] Scale"=180)
     From=?
     To=?
     Angle Type=::NewEnumerator0
     Coordinate Space=::Local
     Normalize Vectors=false
    )) {isSetParameters=true}
  ]
  rendererClasses [&NiagaraSpriteRendererProperties]
  renderers [
   {rendererClass=&NiagaraSpriteRendererProperties}
   {rendererClass=&NiagaraSpriteRendererProperties, rendererIndex=1}
  ]
 }
 Glow_Decals {
  emitterUpdateScript [
   &EmitterState()
   &SpawnBurst_Instantaneous(Spawn Group=1, Use Spawn Probability=false, UseLoopCountLimit=true) {module=SpawnBurst_Instantaneous001}
   &SpawnBurst_Instantaneous(Spawn Count=7, Use Spawn Probability=false, UseLoopCountLimit=true)
  ]
  particleSpawnScript [
   /SetVariables(Particles.VisibilityTag=&SpawnGroup()) {isSetParameters=true}
   &PartitionParticles(Number Of Partitions=7)
   &InitializeParticle(
    Color=<rgba 0.0392 0.0116 0.1478 0.8>
    Lifetime=3
    Sprite Rotation Mode=::NewEnumerator0
    Sprite Size Mode=::NewEnumerator3
    Uniform Sprite Size=&Multiply_Float(
     A=&SelectFloatFromArray(
      Direct Array Index=$Particles.PartitionParticles.Partition
      Float Selection Array=DataInterface() {propertyValues {FloatData [50,50,50,35,35,30,30]}}
     )
     B=&MakeCustomFloatFromBool(Bool=$User.IsLargeBody, True Float=1.5, False Float=1)
    )
    UsePositionOffset=false
    Write Material Random=true
   ) {module=InitializeParticle001}
   /SetVariables(
    Particles.CenterDarkGlowColor:LinearColor=&MultiplyLinearColor(
     Linear Color=<rgba 0.0392 0.0116 0.1478 0.9>
     Scale Factor=&MakeVector4(
      X=1
      Y=1
      Z=1
      W=&FNTimeOfDayScalar(Night Scale=1, Morning Scale=0.8, Day Scale=0.7, Evening Scale=1)
     )
    )
    Particles.CenterGlowColor:LinearColor=/Niagara/DynamicInputs/LinearColor/MakeLinearColorFromVector4(Vector:Vector4f=&MakeVector4FromVectorAndFloat(
      "Vector (RGB)"=<4 1.9574 0.4883>
      "Float (Alpha)"=&Multiply_Float(A=0.6, B=&FNTimeOfDayScalar(Night Scale=0.7, Morning Scale=1, Evening Scale=0.8))
     ))
    Particles.DecalColor:LinearColor=&Multiply_LinearColorByFloat(
     Linear Color=&SelectLinearColorFromArray(
      Array Sampling Mode=::NewEnumerator1
      Direct Array Index=$Particles.PartitionParticles.Partition
      Color Selection Array=DataInterface() {propertyValues {ColorData [
         <rgba 1 0.7404 0.6826 1>
         <rgba 1 0.7404 0.6826 1>
         <rgba 1 0.7404 0.6826 1>
         <rgba 1 0.5893 0.4978 1>
         <rgba 1 0.5893 0.4978 1>
         <rgba 1 0.5893 0.4978 1>
         <rgba 1 0.5893 0.4978 1>
        ]}}
      Random Seed=?
     )
     Float=&MakeCustomFloatFromBool(
      Bool=$User.IsFrontEnd
      True Float=2
      False Float=&Multiply_Float(
       A=3
       B=&FNTimeOfDayScalar(Night Scale=0.45, Morning Scale=0.6, Day Scale=0.6, Evening Scale=0.4)
      )
     )
    )
    Particles.coreglowColor:LinearColor=&Multiply_LinearColorByFloat(
     Linear Color=&SelectLinearColorFromArray(
      Array Sampling Mode=::NewEnumerator1
      Direct Array Index=$Particles.PartitionParticles.Partition
      Color Selection Array=DataInterface() {propertyValues {ColorData [
         <rgba 1 0.6726 0.5615 1>
         <rgba 1 0.6726 0.5615 1>
         <rgba 1 0.6726 0.5615 1>
         <rgba 1 0.6726 0.5615 0.8>
         <rgba 1 0.6726 0.5615 0.8>
         <rgba 1 0.6726 0.5615 0.7>
         <rgba 1 0.6726 0.5615 0.7>
        ]}}
      Random Seed=?
     )
     Float=&MakeCustomFloatFromBool(
      Bool=$User.IsFrontEnd
      True Float=0.5
      False Float=&Multiply_Float(
       A=1.5
       B=&FNTimeOfDayScalar(Night Scale=0.5, Morning Scale=0.6, Day Scale=0.6, Evening Scale=0.5)
      )
     )
    )
   ) {isSetParameters=true}
   /SetVariables(
    Particles.CenterGlowScale:Vector2f=&Vector2DFromFloat(Value=&Multiply_Float(A=115, B=&MakeCustomFloatFromBool(Bool=$User.IsLargeBody, True Float=1.4, False Float=1)))
    Particles.centerDarkGlowScale:Vector2f=&Vector2DFromFloat(Value=&Multiply_Float(A=125, B=&MakeCustomFloatFromBool(Bool=$User.IsLargeBody, True Float=1.4, False Float=1)))
    Particles.glowScale:Vector2f=&Vector2DFromFloat(Value=&Multiply_Float(
      A=&SelectFloatFromArray(
       Direct Array Index=$Particles.PartitionParticles.Partition
       Float Selection Array=DataInterface() {propertyValues {FloatData [150,80,80,65,65,40,40]}}
      )
      B=&MakeCustomFloatFromBool(Bool=$User.IsLargeBody, True Float=1.2, False Float=1)
     ))
   ) {isSetParameters=true, module=SetVariables_#2}
   /SetVariables(Particles.glowCameraOffset=-25) {isSetParameters=true, module=SetVariables_#3}
   &CameraOffset(Camera Offset Amount=5)
  ]
  particleUpdateScript [
   &ParticleState(
    Kill Particles When Lifetime Has Elapsed=false
    Let Infinitely Lived Particles Die When Emitter Deactivates=true
    Loop Particles Lifetime=true
   )
   &ScaleColor(
    Scale Alpha=&FNTimeOfDayScalar(Night Scale=0.8, Morning Scale=1, Evening Scale=0.8)
    ScaleA=true
    ScaleRGB=true
   )
   &SkeletalMeshLocation(
    Bone / Socket Index=&SelectIntFromArray(Direct Array Index=$Particles.PartitionParticles.Partition, Int Selection Array=DataInterface() {propertyValues {IntData [0,1,2,3,4,5,6]}})
    Bone Sampling Mode=::NewEnumerator4
    MaskSpawnGroup=false
    Orientation Sampling=::NewEnumerator1
    Sampled Position Offset=&SelectVectorFromArray(
     Direct Array Index=$Particles.PartitionParticles.Partition
     Vector Selection Array=DataInterface() {propertyValues {FloatData [<0 0 -5>, <0 0 0>, <0 0 0>, <0 0 0>, <0 0 0>, <0 0 0>, <0 0 0>]}}
    )
    Sampled Position Offset Coordinate Space:ENiagaraSkelMeshTransforms=::NewEnumerator2
    Skeletal Mesh=$System.Skeletal Mesh
   ) {module=SkeletalMeshLocation003}
   &SkeletalMeshLocation(
    Bone / Socket Index=&SelectIntFromArray(Direct Array Index=0, Int Selection Array=DataInterface() {propertyValues {IntData [1]}})
    Bone Sampling Mode=::NewEnumerator4
    MaskSpawnGroup=false
    Orientation Sampling=::NewEnumerator1
    Position Sampling=::NewEnumerator2
    Sampled Position Offset Coordinate Space:ENiagaraSkelMeshTransforms=::NewEnumerator2
    Skeletal Mesh=$System.Skeletal Mesh
   )
   &SkeletalMeshLocation(
    Bone / Socket Index=&SelectIntFromArray(Direct Array Index=0, Int Selection Array=DataInterface() {propertyValues {IntData [2]}})
    Bone Sampling Mode=::NewEnumerator4
    MaskSpawnGroup=false
    Orientation Sampling=::NewEnumerator1
    Position Sampling=::NewEnumerator2
    Sampled Position Offset Coordinate Space:ENiagaraSkelMeshTransforms=::NewEnumerator2
    Skeletal Mesh=$System.Skeletal Mesh
    Spawn Group Mask=1
   ) {module=SkeletalMeshLocation001}
   /SetVariables(Particles.CenteredGlowPosition:NiagaraPosition=&LerpPosition(A=$Output.SkeletalMeshLocation.SampledPositionWithoutOffset, B=$Output.SkeletalMeshLocation001.SampledPositionWithoutOffset, Alpha=0.5)) {isSetParameters=true}
   &DynamicMaterialParameters(
    Index 0 Param 1=&SelectFloatFromArray(
     Direct Array Index=$Particles.PartitionParticles.Partition
     Float Selection Array=DataInterface() {propertyValues {FloatData [5,3,3,4,4,4,4]}}
    )
    Index 0 Param 3=1
    Module.Param0WriteEnabled=true
    Module.Param1WriteEnabled=true
    Module.Param2WriteEnabled=false
    Module.Param3WriteEnabled=false
   )
   /SetVariables(Particles.DynamicParameterCenteredGlow:Vector4f=&MakeVector4(X=2, Y=1, Z=0, W=1)) {isSetParameters=true}
   &SpriteRotationRate(Rotation Rate=10)
  ]
  rendererClasses [&NiagaraSpriteRendererProperties]
  renderers [
   {rendererClass=&NiagaraSpriteRendererProperties}
   {rendererClass=&NiagaraSpriteRendererProperties, rendererIndex=1}
   {rendererClass=&NiagaraSpriteRendererProperties, rendererIndex=2}
   {rendererClass=&NiagaraSpriteRendererProperties, rendererIndex=3}
   {rendererClass=&NiagaraSpriteRendererProperties, rendererIndex=4}
  ]
 }
 Large_flares {
  emitterUpdateScript [
   &EmitterState()
   &SpawnRate(SpawnRate=&RandomRangeFloat(Minimum=0.25, Maximum=0.75), Use Spawn Probability=false)
   &SpawnBurst_Instantaneous(Spawn Count=4, Use Spawn Probability=false, UseLoopCountLimit=true) {enabled=false}
  ]
  particleSpawnScript [
   &PartitionParticles(Number Of Partitions=4)
   &InitializeParticle(
    Color=&SelectLinearColorFromArray(
     Array Sampling Mode=::NewEnumerator1
     Direct Array Index=0
     Color Selection Array=DataInterface() {propertyValues {ColorData [<rgba 1.5 0.6917 0.4856 1>, <rgba 1.5 1.1073 0.4885 1>]}}
     Random Seed=?
    )
    Lifetime=&RandomRangeFloat(Minimum=0.6, Maximum=1.2)
    Sprite Rotation Mode=::NewEnumerator2
    Sprite Size Mode=::NewEnumerator3
    Uniform Sprite Size=&Multiply_Float(
     A=&SelectFloatFromArray(
      Array Sampling Mode=::NewEnumerator0
      Direct Array Index=?
      Float Selection Array=DataInterface() {propertyValues {FloatData [80,85,90]}}
      Random Seed=0
     )
     B=&MakeCustomFloatFromBool(Bool=$User.IsLargeBody, True Float=1.1, False Float=1)
    )
    UsePositionOffset=false
    Write Material Random=true
   ) {module=InitializeParticle001}
   /SetVariables(
    Particles.CoronaScale:Vector2f=&Multiply_Vector2D_ByFloat(Float=1.1, Vector2D=$Particles.SpriteSize)
    Particles.FlareLineScale:Vector2f=&MakeVector2D(
     X=&Multiply_Float(A=&MakeFloatFromVector2d(Vector 2D=$Particles.SpriteSize), B=3)
     Y=10
    )
    Particles.glowScale:Vector2f=&Multiply_Vector2D_ByFloat(Float=2, Vector2D=$Particles.SpriteSize)
   ) {isSetParameters=true, module=SetVariables_#2}
   /SetVariables(
    Particles.CoronaColor:LinearColor=<rgba 0.162 0.0343 0.063 1>
    Particles.FlareLineColor:LinearColor=<rgba 1 0.7994 0.583 1>
    Particles.glowcolor:LinearColor=<rgba 0.162 0.0343 0.063 1>
   ) {isSetParameters=true, module=SetVariables_#4}
   &CameraOffset(Camera Offset Amount=-45)
   &AddVelocity(Velocity=&RandomRangeVector(Minimum=<-1 -1 -1>, Maximum=<1 1 1>), Velocity Speed Scale=5)
   &ApplyInitialForces()
   /SetVariables() {isSetParameters=true}
  ]
  particleUpdateScript [
   &ParticleState()
   &ScaleSpriteSize(Uniform Curve Sprite Scale=DataInterface() {propertyValues {Curve {
       defaultValue=0
       keys [
        {arriveTangent=14.032, interpMode=RCIM_Cubic, leaveTangent=14.032, tangentMode=RCTM_User, time=0, value=0}
        {arriveTangent=0.7062, interpMode=RCIM_Cubic, leaveTangent=0.7062, tangentMode=RCTM_User, time=0.1, value=0.85}
        {arriveTangentWeight=0.4228, interpMode=RCIM_Cubic, tangentMode=RCTM_User, time=1, value=1.15}
       ]
      }}})
   &DynamicMaterialParameters(
    Index 0 Param 1=&SelectFloatFromArray(
     Direct Array Index=$Particles.PartitionParticles.Partition
     Float Selection Array=DataInterface() {propertyValues {FloatData [5,3,3,4,4,4,4]}}
    )
    Index 0 Param 2=10
    Index 0 Param 3=1
    Module.Param0WriteEnabled=true
    Module.Param1WriteEnabled=true
    Module.Param2WriteEnabled=false
    Module.Param3WriteEnabled=false
   )
   &SkeletalMeshLocation(
    Bone / Socket Index=&SelectIntFromArray(Direct Array Index=$Particles.PartitionParticles.Partition, Int Selection Array=DataInterface() {propertyValues {IntData [0,1,0,2]}})
    Bone Sampling Mode=::NewEnumerator4
    MaskSpawnGroup=false
    Orientation Sampling=::NewEnumerator1
    Sampled Position Offset=&SelectVectorFromArray(
     Direct Array Index=$Particles.PartitionParticles.Partition
     Vector Selection Array=DataInterface() {propertyValues {FloatData [<-20 0 25>, <22 0 -35>, <25 0 18>, <-20 0 -25>]}}
    )
    Sampled Position Offset Coordinate Space:ENiagaraSkelMeshTransforms=::NewEnumerator2
    Skeletal Mesh=$System.Skeletal Mesh
   ) {module=SkeletalMeshLocation003}
   &InterpolateOverTime(Target Position Value:NiagaraPosition=$Particles.Position, Variable Type=::NewEnumerator6)
   &SpriteRotationRate(Rotation Rate=&FloatFromCurve(
     FloatCurve=DataInterface() {propertyValues {Curve {
        defaultValue=1
        keys [{interpMode=RCIM_Cubic, tangentMode=RCTM_User, time=0, value=0}, {interpMode=RCIM_Cubic, tangentMode=RCTM_User, time=1, value=1}]
       }}}
     Scale Curve=25
    ))
   &SolveForcesAndVelocity(Clamp Velocity=false, Limit Acceleration=false)
   &ScaleColor(
    Scale Alpha=&Multiply_Float(
     A=&FloatFromCurve(FloatCurve=DataInterface() {propertyValues {Curve {keys [<at 0.0028 0>,<at 0.0949 1>,<at 0.645 1>,<at 0.8541 0.4>,<at 1 0>]}}})
     B=&MakeCustomFloatFromBool(Bool=$User.IsFrontEnd, True Float=0.9, False Float=1)
    )
    Scale RGB=&VectorFromCurve(
     VectorCurve=DataInterface() {propertyValues {
       XCurve {defaultValue=1, keys [<at 0 1>,<at 0.0871 1.5>,<at 0.543 1>,<at 1 1>]}
       YCurve {defaultValue=1, keys [<at 0 1>,<at 0.0871 1.5>,<at 0.543 1>,<at 1 0.6358>]}
       ZCurve {defaultValue=1, keys [<at 0 1>,<at 0.0871 1.5>,<at 0.543 1>,<at 1 0.3847>]}
      }}
     Scale Curve=&Multiply_VectorByFloat(
      Vector=<1 1 1>
      Float=&FNTimeOfDayScalar(Night Scale=0.9, Morning Scale=1.1, Day Scale=1.2, Evening Scale=1)
     )
    )
    ScaleA=true
    ScaleRGB=true
   )
   /SetVariables(Particles.Position=$StackContext.InterpolateOverTime.InterpolateOverTime_Position.Moving Average) {isSetParameters=true, module=SetVariables_#5}
   /SetVariables(
    Particles.CoronaColor:LinearColor=&ColorFromCurve(ColorCurve=DataInterface() {propertyValues {
       AlphaCurve {keys [
         <at 0.0003 0>
         {arriveTangent=3.3163, interpMode=RCIM_Cubic, leaveTangent=3.3163, time=0.0898, value=1}
         <at 0.3019 1>
         <at 0.7246 0 RCIM_Cubic>
        ]}
       BlueCurve {keys [<at 0 0.0709 RCIM_Cubic>,<at 1 0.0366 RCIM_Cubic>]}
       GreenCurve {keys [<at 0 0.0576 RCIM_Cubic>,<at 1 0.0234 RCIM_Cubic>]}
       RedCurve {keys [<at 0 0.3659 RCIM_Cubic>,<at 1 0.1302 RCIM_Cubic>]}
      }})
    Particles.FlareLineColor:LinearColor=&ColorFromCurve(ColorCurve=DataInterface() {propertyValues {
       AlphaCurve {keys [<at 0.0008 0>, {arriveTangent=0.9639, interpMode=RCIM_Cubic, leaveTangent=0.9639, time=0.078, value=1}, <at 0.4158 0.4>, <at 1 0 RCIM_Cubic>]}
       BlueCurve {keys [<at 0.1472 0.2812 RCIM_Cubic>,<at 0.4097 0.0549 RCIM_Cubic>,<at 1 0.0645>]}
       GreenCurve {keys [<at 0.1472 0.6747 RCIM_Cubic>,<at 0.4097 0.1082 RCIM_Cubic>,<at 1 0.0221>]}
       RedCurve {keys [<at 0.1472 1 RCIM_Cubic>,<at 0.4097 0.5547 RCIM_Cubic>,<at 1 0.2292>]}
      }})
    Particles.glowcolor:LinearColor=&MultiplyLinearColor(
     Linear Color=&ColorFromCurve(ColorCurve=DataInterface() {propertyValues {
        AlphaCurve {keys [
          <at 0.0003 0>
          {arriveTangent=2.5573, interpMode=RCIM_Cubic, leaveTangent=2.5573, time=0.0898, value=1.1}
          <at 0.3914 1>
          <at 1 0 RCIM_Cubic>
         ]}
        BlueCurve {keys [<at 0 0.0466 RCIM_Cubic>,<at 1 0.0588 RCIM_Cubic>]}
        GreenCurve {keys [<at 0 0.0616 RCIM_Cubic>,<at 1 0.0246 RCIM_Cubic>]}
        RedCurve {keys [<at 0 0.3047 RCIM_Cubic>,<at 1 0.1302 RCIM_Cubic>]}
       }})
     Scale Factor=&MakeVector4(
      X=1
      Y=1
      Z=1
      W=&MakeCustomFloatFromBool(
       Bool=$User.IsFrontEnd
       True Float=0.8
       False Float=&FNTimeOfDayScalar(Night Scale=1, Morning Scale=1, Day Scale=0.8, Evening Scale=1)
      )
     )
    )
   ) {isSetParameters=true}
   /SetVariables(Particles.CoronMatParams:Vector4f=&MakeVector4(
     X=&FloatFromCurve(FloatCurve=DataInterface() {propertyValues {Curve {
         defaultValue=0
         keys [
          {arriveTangent=-3.2694, interpMode=RCIM_Cubic, leaveTangent=-3.2694, tangentMode=RCTM_User, time=0, value=1}
          {arriveTangentWeight=0.1357, interpMode=RCIM_Cubic, leaveTangentWeight=0.1357, tangentMode=RCTM_User, time=0.3, value=0.55}
          <at 0.6 0.45 RCIM_Cubic>
         ]
        }}})
     Y=1
     Z=1
     W=1
    )) {isSetParameters=true}
   /SetVariables(Particles.pivotoffsetline:Vector2f=&MakeVector2D(
     X=0.5
     Y=&FloatFromCurve(FloatCurve=DataInterface() {propertyValues {Curve {keys [{arriveTangent=0.4709, interpMode=RCIM_Cubic, leaveTangent=0.4709, tangentMode=RCTM_User, time=0, value=0.5}, <at 1 0.8 RCIM_Cubic>]}}})
    )) {isSetParameters=true}
  ]
  rendererClasses [&NiagaraSpriteRendererProperties]
  renderers [
   {rendererClass=&NiagaraSpriteRendererProperties}
   {rendererClass=&NiagaraSpriteRendererProperties, rendererIndex=1}
   {rendererClass=&NiagaraSpriteRendererProperties, rendererIndex=2}
   {rendererClass=&NiagaraSpriteRendererProperties, rendererIndex=3}
   {rendererClass=&NiagaraSpriteRendererProperties, rendererIndex=4}
  ]
 }
}
summary {
 systemName=NS_Aura_CenterStage
 emitters [
  {emitterName=Confetti, rendererClasses [&NiagaraSpriteRendererProperties]}
  {emitterName=Large_flares, rendererClasses [&NiagaraSpriteRendererProperties]}
  {emitterName=Cross_Flares, rendererClasses [&NiagaraSpriteRendererProperties]}
  {emitterName=Glow_Decals, rendererClasses [&NiagaraSpriteRendererProperties]}
 ]
 userVariables [
  User.bIsFrontEnd {type=NiagaraBool}
  User.bIsFrontEndPreview {type=NiagaraBool}
  User.Interpolate Intensity {type=NiagaraFloat, defaultValue=5, description=""}
  User.IsFrontEnd {type=NiagaraBool, defaultValue=true}
  User.IsLargeBody {type=NiagaraBool}
 ]
}
